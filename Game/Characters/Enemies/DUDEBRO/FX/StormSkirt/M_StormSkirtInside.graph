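# Masked, two-sided lit material for the inside of the DUDEBRO enemy's storm skirt effect, usable on skeletal meshes.
# Opacity combines a V-gradient eroded by scrolling smoke, a panned storm edge mask, scrolling noise, and side-dependent Fresnel times red vertex color, dithered through TemporalAA.
# The same two noise samples drive world position offset.
# Emissive blends two dark blue colors by noise, shaded by view-space reflection of a panned, flattenable smoke normal map.
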Material /Game/Characters/Enemies/DUDEBRO/FX/StormSkirt/M_StormSkirtInside
graph {
 properties {
  bUsedWithMeshDeformer=true
  bUsedWithSkeletalMesh=true
  blendMode=BLEND_Masked
  pixelDepthOffsetMode=PDOM_Legacy
  twoSided=true
  usedShadingModels=Default Lit
 }
 nodes [
  n1 = TextureSample(UVs=Panner(Coordinate=TextureCoordinate() {vTiling=3}) {speedX=1, speedY=0.3}) {samplerType=SAMPLERTYPE_Normal, texture=/Game/Effects/Fort_Effects/Gameplay/SmokeDustExplosions/Textures/T_Explosion_NM_01}
  MaterialFunctionCall(Normal=@n1.RGB, Flatness=ScalarParameter() {parameterName=FlatternNormal}) {materialFunction=&FlattenNormal}
  n2 = MaterialFunctionCall(Float2=Saturate(None=TextureCoordinate() {vTiling=6.5})) {materialFunction=&BreakOutFloat2Components}
  n3 = OneMinus(None=MaterialFunctionCall(Float2=TextureCoordinate()) {materialFunction=&BreakOutFloat2Components}.G)
  n4 = TextureCoordinate()
  n5 = Subtract(
   A=Multiply(A=Multiply(A=@n2.G), B=Multiply(A=@n3, B=@n3))
   B=TextureSample(UVs=Panner(Coordinate=@n4) {speedX=2, speedY=0.3}) {samplerType=SAMPLERTYPE_LinearColor, texture=/Game/Global/Textures/Default/Tileable_Patterns/T_Tile_Smoke_Small}.R
  )
  n6 = Panner(Coordinate=TextureCoordinate()) {speedX=1}
  n7 = Multiply(
   A=Saturate(None=Add(A=@n5) {constB=0.8})
   B=TextureSampleParameter2D(UVs=MaterialFunctionCall(XY=@n6) {materialFunction=&Swizzle}.YX) {parameterName=Mask, texture=/Game/Effects/Fort_Effects/Textures/Mask/T_StormEdgeMask}.R
  )
  n8 = ScalarParameter() {defaultValue=1, parameterName=EdgeFadeExp}
  n9 = MaterialFunctionCall!(Axis, Exponent=@n8, Normal=Multiply(A=VertexNormalWS()) {constB=-1}) {materialFunction=/Game/Effects/Fort_Effects/Materials/_Functions/MF_AxisAlignedFrensnel}
  n10 = MaterialFunctionCall(Exponent=@n8) {materialFunction=&MF_AxisAlignedFrensnel}
  n11 = TwoSidedSign()
  n12 = MaterialFunctionCall(Alpha Threshold=Multiply(A=LinearInterpolate(A=@n9.Fresnel, B=@n10.Fresnel, Alpha=@n11), B=VertexColor().R)) {materialFunction=&DitherTemporalAA}
  n13 = TextureSample(UVs=Panner(Coordinate=TextureCoordinate() {vTiling=3.5}) {speedX=0.5, speedY=0.5}) {samplerType=SAMPLERTYPE_LinearColor, texture=&T_NoisePack_LinearColor}
  n14 = TextureSample(UVs=Panner(Coordinate=TextureCoordinate() {vTiling=3}) {speedX=1, speedY=0.8}) {samplerType=SAMPLERTYPE_LinearColor, texture=&T_NoisePack_LinearColor}
  Multiply(A=Multiply(A=Multiply(A=@n7, B=@n12.Result), B=Add(A=@n13.G, B=@n14.R)), B=ScalarParameter() {defaultValue=1, parameterName=OpacityOverall})
  Multiply(A=Multiply(A=@n13.RGB, B=@n14.RGB), B=ScalarParameter() {defaultValue=40, parameterName=WPO_Intensity}) {constB=40}
  n15 = ComponentMask(None=Transform(None=ReflectionVectorWS(CustomWorldNormal=@n1.RGB)) {transformType=TRANSFORM_View}) {g=true}
  n16 = VectorParameter() {parameterName=Color1, defaultValue=<rgba 0.0069 0.0093 0.0156 1>}
  n17 = VectorParameter() {parameterName=Color2, defaultValue=<rgba 0.0652 0.1041 0.1406 1>}
  Multiply(
   A=Multiply(A=LinearInterpolate(Alpha=@n15) {constA=0.7}, B=LinearInterpolate(A=@n16.RGB, B=@n17.RGB, Alpha=@n13.R))
   B=ScalarParameter() {defaultValue=0.6, parameterName=EmissiveIntensity}
  )
  ScalarParameter() {defaultValue=1, parameterName=Roughness}
  ScalarParameter() {defaultValue=0.5, parameterName=Metallic}
  ScalarParameter() {parameterName=Specular}
 ]
}
