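# Masked, two-sided subsurface mesh-particle material for the smoky, lightning-laced strips of the Storm King's storm skirt.
# World position offset combines noise, a sine/cosine swirl and particle-fed socket offsets, then widens strips camera-facing via cross product.
# Lifetime curve-atlas ramps and per-particle random drive motion, color and panning smoke opacity.
# AsTranslucent switch selects MF_DudebroSkinFX with depth fade and plasma emissive, or dithered masking with pixel depth offset.
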Material /Game/Characters/Enemies/DUDEBRO/FX/StormSkirt/M_StormStrip
graph {
 properties {
  bUseMaterialAttributes=true
  bUsedWithMeshParticles=true
  blendMode=BLEND_Masked
  pixelDepthOffsetMode=PDOM_Legacy
  shadingModel=MSM_Subsurface
  twoSided=true
  usedShadingModels=Subsurface
 }
 nodes [
  Constant3Vector() {constant=<rgba 0 0 -912.0549 0>}
  Constant3Vector() {constant=<rgba 0 0 -1471.3365 0>}
  n1 = Constant3Vector() {constant=<rgba 1 0 0 0>}
  n2 = WorldPosition()
  n3 = VectorParameter() {parameterName=UV, defaultValue=<rgba 1 0.5 0.4 -0.1>}
  n4 = Multiply(A=TextureCoordinate(), B=AppendVector(A=@n3.R, B=@n3.R))
  n5 = Add(A=Time(), B=Multiply(A=DynamicParameter() {paramNames [BP_Whip_Mid_X, BP_Whip_Mid_Y, BP_Whip_Mid_Z, Random]}.Random) {constB=128})
  n6 = AppendVector(A=@n3.B, B=@n3.B)
  n7 = Subtract(A=TextureSample(UVs=Panner(Coordinate=@n4, Time=@n5, Speed=@n6)) {samplerType=SAMPLERTYPE_LinearColor, texture=&T_NoisePack_LinearColor}.RGB) {constB=0.5}
  n8 = VectorParameter() {parameterName=WPO_Mutiply, defaultValue=<rgba 300 300 500 1>}
  n9 = CurveAtlasRowParameter(CurveTime=ParticleRelativeTime()) {
   atlas=&AthenaCurveAtlas01
   curve=/Game/Athena/LinearColorCurves/Curve_StormKing_StormSkirt_Ramp2
   defaultValue=37
   parameterName=WPORamp
  }
  n10 = CurveAtlasRowParameter(CurveTime=ParticleRelativeTime()) {
   atlas=&AthenaCurveAtlas01
   curve=/Game/Athena/LinearColorCurves/Curve_StormKing_StormSkirt_Ramp
   defaultValue=34
   parameterName=TornadoRamp
  }
  n11 = Add(A=@n10.B, B=Multiply(A=Time(), B=ScalarParameter() {defaultValue=0.5, parameterName=Speed}))
  n12 = Multiply(
   A=AppendVector(A=Sine(None=@n11), B=Cosine(None=@n11))
   B=Multiply(A=ScalarParameter() {defaultValue=200, parameterName=WaveOffset}, B=@n10.A)
  )
  n13 = Constant()
  n14 = CurveAtlasRowParameter(CurveTime=ParticleRelativeTime()) {
   atlas=&AthenaCurveAtlas01
   curve=&Curve_StormKing_StormSkirt_Ramp2
   defaultValue=37
   parameterName=SocketAnimationWPORamp
  }
  n15 = Multiply(
   A=Add(A=ParticleColor().RGB, B=VectorParameter() {parameterName=SocketAnimationInitialOffset_Top, defaultValue=<rgba 0 0 -1177.0692 0>}.RGB)
   B=@n14.G
  )
  n16 = DynamicParameter() {paramNames [BP_Whip_Mid_X, BP_Whip_Mid_Y, BP_Whip_Mid_Z, Random]}
  n17 = Multiply(
   A=Add(
    A=MaterialFunctionCall(X=@n16.BP_Whip_Mid_X, Y=@n16.BP_Whip_Mid_X, Z=@n16.BP_Whip_Mid_X) {materialFunction=&MakeFloat3}.Result
    B=VectorParameter() {parameterName=SocketAnimationInitialOffset_Mid, defaultValue=<rgba 0 0 -912.0549 0>}.RGB
   )
   B=@n14.B
  )
  n18 = Add(A=Multiply(A=Multiply(A=@n7, B=@n8.RGB), B=@n9.R), B=Add(A=AppendVector(A=@n12, B=@n13), B=Add(A=@n15, B=@n17)))
  n19 = Normalize(VectorInput=Subtract(A=Add(A=@n2.XYZ, B=@n18), B=CameraPositionWS()))
  n20 = ComponentMask(None=TextureCoordinate()) {g=true}
  n21 = Multiply(
   A=CrossProduct(A=Transform(None=@n1), B=@n19)
   B=Multiply(A=Subtract(A=@n20) {constB=0.5}, B=ScalarParameter() {defaultValue=128, parameterName=Stripwide}) {constB=200}
  )
  n22 = VectorParameter() {parameterName=OpacityUV, defaultValue=<rgba 3 0.4 1 0.5>}
  n23 = Panner(
   Coordinate=Multiply(A=TextureCoordinate(), B=AppendVector(A=@n22.R, B=@n22.R))
   Time=Add(A=Time(), B=DynamicParameter() {paramNames [BP_Whip_Mid_X, BP_Whip_Mid_Y, BP_Whip_Mid_Z, Random]}.Random)
   Speed=AppendVector(A=@n22.B, B=@n22.B)
  )
  n24 = Multiply(
   A=TextureSample(UVs=@n23) {samplerType=SAMPLERTYPE_LinearColor, texture=/Game/Global/Textures/Default/Tileable_Patterns/T_Tile_Smoke_Small}.R
   B=TextureSample() {texture=/Game/Effects/Fort_Effects/Textures/Ramps/T_Ramp_Smooth}.R
  )
  n25 = Multiply(A=ScalarParameter() {defaultValue=1.2, parameterName=Opacity}, B=ParticleColor().A)
  n26 = CurveAtlasRowParameter(CurveTime=ParticleRelativeTime()) {
   atlas=&AthenaCurveAtlas01
   curve=&Curve_StormKing_StormSkirt_Ramp
   defaultValue=34
   parameterName=OpacityMask
  }
  n27 = DotProduct(A=@n19, B=VertexNormalWS())
  n28 = SetMaterialAttributes!(
   MaterialAttributes=SetMaterialAttributes!(MaterialAttributes, World Position Offset=Add(A=@n21, B=@n18)) {attributeSetTypes [#1]}
   Emissive Color
   Opacity Mask=Multiply(A=Multiply(A=@n24, B=@n25) {constB=1.2}, B=@n26.R)
   Opacity=@n27
  ) {attributeSetTypes [#2, #3, #4]}
  n29 = CurveAtlasRowParameter(CurveTime=DynamicParameter() {paramNames [BP_Whip_Mid_X, BP_Whip_Mid_Y, BP_Whip_Mid_Z, Random]}.Random) {
   atlas=&AthenaCurveAtlas01
   curve=/Game/Athena/LinearColorCurves/Curve_StormKing_StormSkirt_Color
   defaultValue=35
   parameterName=BaseColor
  }
  n30 = VectorParameter() {parameterName=BottomColor, defaultValue=<rgba 0.1047 0.0286 0.1406 1>}
  n31 = OneMinus(None=ParticleRelativeTime())
  n32 = Multiply(A=@n31, B=@n31)
  n33 = Multiply(A=@n32, B=@n32)
  n34 = TextureSample(UVs=Multiply(A=@n23, B=Constant2Vector() {g=1, r=0.3333})) {texture=/Game/Global/Textures/Default/Tileable_Patterns/T_Smoke_Puff_Tile_Up}
  n35 = Multiply(
   A=LinearInterpolate(Alpha=Multiply(A=@n27, B=@n27)) {constA=3, constB=0.75}
   B=Multiply(A=LinearInterpolate(A=@n29, B=@n30.RGB, Alpha=@n33), B=LinearInterpolate(Alpha=@n34.R) {constA=0.25})
  )
  n36 = GetMaterialAttributes(None=SetMaterialAttributes!(
    MaterialAttributes=SetMaterialAttributes!(MaterialAttributes=@n28, Base Color=@n35, Subsurface Color=@n35) {attributeSetTypes [#5, #6]}
    Specular=Constant()
    Metallic=Constant() {r=0.5}
    Roughness=Constant() {r=1}
   ) {attributeSetTypes [#7, #8, #9]}) {attributeGetTypes [#3]}
  n37 = GetMaterialAttributes(None=@n36.MaterialAttributes) {attributeGetTypes [#5]}
  n38 = Multiply(A=@n36.Opacity Mask, B=DepthFade(FadeDistance=ScalarParameter() {defaultValue=100, parameterName=DF_FD}) {fadeDistanceDefault=25})
  n39 = VectorParameter() {parameterName=UV_EmissiveLightning, defaultValue=<rgba 1 1 0.1 0.2>}
  n40 = Multiply(A=TextureCoordinate(), B=AppendVector(A=@n39.R, B=@n39.R))
  n41 = Add(A=Time(), B=Multiply(A=DynamicParameter() {paramNames [BP_Whip_Mid_X, BP_Whip_Mid_Y, BP_Whip_Mid_Z, Random]}.Random) {constB=128})
  n42 = AppendVector(A=@n39.B, B=@n39.B)
  n43 = Constant2Vector() {g=1.512, r=1.215}
  n44 = Multiply(
   A=TextureSample(UVs=Panner(Coordinate=@n40, Time=@n41, Speed=@n42)) {texture=/Game/Effects/Fort_Effects/Textures/Plasma/T_Plasma_Aurora_Seamless}.G
   B=TextureSample(UVs=Multiply(A=@n40, B=@n43)) {texture=&T_Plasma_Aurora_Seamless}.B
  )
  n45 = VectorParameter() {parameterName=EmissiveLightning, defaultValue=<rgba 10 0 5.9829 0>}
  n46 = Add(
   A=Multiply(A=@n37.Base Color, B=ScalarParameter() {defaultValue=2, parameterName=EmissiveBoost})
   B=Multiply(A=Multiply(A=@n44, B=@n45.RGB), B=ParticleRelativeTime())
  )
  n47 = Multiply(
   A=MaterialFunctionCall(Random=DynamicParameter() {paramNames [BP_Whip_Mid_X, BP_Whip_Mid_Y, BP_Whip_Mid_Z, Random]}.Random) {materialFunction=&DitherTemporalAA}.Result
   B=ScalarParameter() {defaultValue=256, parameterName=PDO_Multiplier}
  ) {constB=256}
  n48 = ScalarParameter() {defaultValue=128, parameterName=PDO_Subtracter}
  StaticSwitchParameter(
   True=MaterialFunctionCall!(
    InMA=SetMaterialAttributes!(MaterialAttributes=@n37.MaterialAttributes, Opacity=@n38, Emissive Color=@n46) {attributeSetTypes [#4, #2]}
    HaveMasks=StaticBool()
    "FresnelRevert(WithOutMask)"=StaticBool() {value=true}
    EmissiveMutiply=Constant() {r=0.85}
   ) {materialFunction=/Game/Characters/Enemies/DUDEBRO/FX/SkinFX/MF_DudebroSkinFX}.MA
   False=SetMaterialAttributes!(
    MaterialAttributes=@n36.MaterialAttributes
    Opacity Mask=MaterialFunctionCall(Alpha Threshold=@n36.Opacity Mask) {materialFunction=&DitherTemporalAA}.Result
    Pixel Depth Offset=Subtract(A=@n47, B=@n48) {constB=128}
   ) {attributeSetTypes [#3, #10]}
  ) {parameterName=AsTranslucent?}
 ]
}
